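# Animation Blueprint for the Mechanical Engineer Summer backpack cosmetic, adding a springy bounce and continuously spinning gears.
# AnimDynamics runs a low-gravity linear spring on dyn_master_jnt, allowing X/Z travel with damped component-motion influence.
# Each update derives constant-rate rotators from real time, independent of character state.
# Transform Bone nodes then apply them in bone space: a replace plus additive spin on the inner gear, one additive spin on the outer.
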AnimBlueprint /BRCosmetics/Accessories/FORT_Backpacks/Backpack_F_MED_Mechanical_Engineer_Summer/Meshes/F_MED_Mechanical_Engineer_Summer_Pack_AnimBP
target_skeleton=&/F_MED_Mechanical_Engineer_Summer_Pack_Skeleton
graphs [
 AnimGraph {nodes [
   n1 = Anim Dynamics - Bone: dyn_master_jnt(ComponentPose=Local To Component(LocalPose=Input Pose()), Alpha=1, ExternalForce="0,0,0") {
    bGravityOverrideInSimSpace=true
    bLinearSpring=true
    gravityScale=0.1
    linearSpringConstant=20
    alphaBoolBlend {blendOption=Linear}
    alphaScaleBiasClamp {scale=0.1}
    boundBone {boneName=dyn_master_jnt}
    componentLinearAccScale {x=0.12, z=0.12}
    componentLinearVelScale {x=0.12, z=0.12}
    physicsBodyDefinitions [{
      sphereCollisionRadius=0
      boundBone {boneName=dyn_master_jnt}
      boxExtents=<1 1 1>
      constraintSetup {angularLimitsMax=<0 0 0>, angularLimitsMin=<0 0 0>, angularTarget=<0 0 0>, linearAxesMax=<500 0 500>, linearAxesMin=<-500 0 -500>}
      localJointOffset=<0 0 0>
     }]
   }
   n2 = "Transform (Modify) Bone - Bone: dyn_innerGear_rotation"(ComponentPose=@n1, Alpha=1, Translation="0,0,0", Rotation=Get smoothRotatorZ() -> (smoothRotatorZ:Rotator), Scale="1,1,1") {rotationMode=BMM_Replace, rotationSpace=BCS_BoneSpace, alphaBoolBlend {blendOption=Linear}, boneToModify {boneName=dyn_innerGear_rotation}}
   n3 = "Transform (Modify) Bone - Bone: dyn_innerGear_rotation"(ComponentPose=@n2, Alpha=1, Translation="0,0,0", Rotation=Get smoothRotator inner() -> (smoothRotator inner:Rotator), Scale="1,1,1") {rotationMode=BMM_Additive, rotationSpace=BCS_BoneSpace, alphaBoolBlend {blendOption=Linear}, boneToModify {boneName=dyn_innerGear_rotation}}
   n4 = "Transform (Modify) Bone - Bone: dyn_outerGear_rotation"(ComponentPose=@n3, Alpha=1, Translation="0,0,0", Rotation=Get smoothRotator outer() -> (smoothRotator outer:Rotator), Scale="1,1,1") {rotationMode=BMM_Additive, rotationSpace=BCS_BoneSpace, alphaBoolBlend {blendOption=Linear}, boneToModify {boneName=dyn_outerGear_rotation}}
   Output Pose(Result=Component To Local(ComponentPose=@n4)) {layerGroup=None}
  ]}
 EventGraph {nodes [
   Event BlueprintUpdateAnimation()
   > n1 = Sequence!(execute) -> (then_0, then_1, then_2:Exec)
   n2 = GetRealTimeSeconds()
   > Set smoothRotator inner(smoothRotator inner:Rotator=MakeRotator(Roll=float * float(A=@n2, B=100))) -> (Output_Get:Rotator)
   n3 = GetRealTimeSeconds()
   Set smoothRotatorZ(execute=@n1.then_1, smoothRotatorZ:Rotator=MakeRotator(Yaw=float * float(A=@n3, B=51))) -> (Output_Get:Rotator)
   n4 = GetRealTimeSeconds()
   Set smoothRotator outer(execute=@n1.then_2, smoothRotator outer:Rotator=MakeRotator(Pitch=float * float(A=@n4, B=50))) -> (Output_Get:Rotator)
   TryGetPawnOwner()
  ]}
]
variables [
 clickyRotator {type=Rotator}
 smoothRotator inner {type=Rotator}
 smoothRotator outer {type=Rotator}
 smoothRotatorZ {type=Rotator}
]
